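# Niagara skydiving contrail for the OrderGuard cosmetic, attaching effects to two skeletal mesh bones, apparently the hands, split by spawn group.
# Spinning blade meshes with flickering blue lights combine with lightning ribbons, balls and sprites, plus orange spark meshes and GPU sparks.
# Dive state speeds blade spin and ribbon panning, while front-end modifiers adjust some emitters.
# Sparks_Optimized samples both bones once per emitter, duplicating the disabled per-particle Sparks emitter.
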NiagaraSystem /BRCosmetics/Athena/Effects/Particles/Skydiving/NS_Contrail_OrderGuard/NS_Contrail_OrderGuard
dependencies {
 usedDataInterfaces [&NiagaraDataInterfaceCurve, &NiagaraDataInterfaceVectorCurve]
 usedDynamicInputs [&UniformRangedFloat, &MakeCustomFloatFromBool]
 usedModules [
  &SystemState
  /Game/Effects/Niagara/Modules/Skydiving/FN_DivingAttributes/v6/FN_DivingAttributes
  &self:SystemSpawnScript.SystemScriptSource.SystemScriptGraph.NiagaraNodeAssignment_0.SetVariables_#1
 ]
}
emitters {
 Blades {
  emitterUpdateScript [
   &EmitterState(Inactive Response=::NewEnumerator1, Life Cycle Mode=::NewEnumerator1, Loop Duration=0, UseLoopDelay=false)
   &SpawnBurst_Instantaneous(Spawn Count=2, Use Spawn Probability=false, UseLoopCountLimit=true)
  ]
  particleSpawnScript [
   &InitializeParticle(
    Color Mode=::NewEnumerator0
    Lifetime=5
    Mesh Scale Mode=::NewEnumerator3
    Mesh Uniform Scale=1.75
    UsePositionOffset=false
    Write Material Random=true
   )
   &FN_SetSpawnGroup(Number of Spawn Groups=2)
   /SetVariables() {isSetParameters=true}
   /SetVariables() {isSetParameters=true}
  ]
  particleUpdateScript [
   &ParticleState(Kill Particles When Lifetime Has Elapsed=false)
   &SkeletalMeshLocation(
    Bone / Socket Index=&SpawnGroup()
    Bone Sampling Mode=::NewEnumerator4
    MaskSpawnGroup=false
    Orientation Sampling=::NewEnumerator1
    Sampled Position Offset=&MaskVectorBySpawnGroup(Masked Vector=<12 -10 0>, Passthrough Vector=<-12 10 1>)
    Skeletal Mesh=$User.SkeletalMesh
   )
   /SetVariables(Particles.MeshOrientation=&MultiplyQuaternion(
     Quaternion A=$Particles.SkeletalMeshLocation.SampledRotation
     Quaternion B=&MultiplyQuaternion(
      Quaternion A=&MakeQuatFromAxisAngle(Axis=<1 0 0>, AngleInDegrees=-90)
      Quaternion B=&MakeQuatFromAxisAngle(Axis=<0 1 0>, AngleInDegrees=&MaskFloatBySpawnGroup(Masked Float=0, Passthrough Float=180))
     )
    )) {isSetParameters=true}
   /SetVariables(Particles.BladeSpinQuat:Quat4f=&MultiplyQuaternion(
     Quaternion A=$Particles.BladeSpinQuat
     Quaternion B=&MakeQuatFromAxisAngle(
      Axis=&MaskVectorBySpawnGroup(Masked Vector=<0 0 1>, Passthrough Vector=<0 0 -1>)
      AngleInDegrees=&Multiply_Float(A=$Engine.DeltaTime, B=&Lerp_Float(A=200, B=850, Alpha=$System.Dive))
     )
    )) {isSetParameters=true}
   /SetVariables(Particles.BladeQuat:Quat4f=&MultiplyQuaternion(
     Quaternion A=$Particles.MeshOrientation
     Quaternion B=&MultiplyQuaternion(
      Quaternion A=$Particles.BladeSpinQuat
      Quaternion B=&MakeQuatFromAxisAngle(Axis=<0 1 0>, AngleInDegrees=&MaskFloatBySpawnGroup(Masked Float=180, Passthrough Float=0))
     )
    )) {isSetParameters=true}
   &FN_Light_Attributes(
    Color=<rgba 0.2257 0.7464 1 1>
    Create Color Attribute=true
    Create Position Attribute=true
    Create Radius Attribute=true
    Position=&Add_Vector(A=Linked() {linkedVariable {name=Particles.Position, type=NiagaraPosition}}, B=<0 0 20>)
    Radius=200
    Scale Color=&Waveform(
     Global Phase=&Add_Float(A=$Emitter.Age, B=&MaskFloatBySpawnGroup(Masked Float=0.3, Passthrough Float=1))
     Global Amplitude Scale=15
     "Waveform [1]"=::NewEnumerator0
     "[1] Frequency"=4.5
     "[1] Amplitude Min/Max"=<0.2 1>
    )
   )
  ]
  rendererClasses [&NiagaraMeshRendererProperties, &NiagaraLightRendererProperties]
  renderers [
   {rendererClass=&NiagaraMeshRendererProperties, rendererIndex=0}
   {rendererClass=&NiagaraMeshRendererProperties, rendererIndex=1}
   {rendererClass=&NiagaraLightRendererProperties, rendererIndex=2}
  ]
 }
 LightningBalls {
  emitterUpdateScript [
   &EmitterState()
   &SpawnRate(SpawnRate=5, Use Spawn Probability=false)
  ]
  particleSpawnScript [
   &FN_SetSpawnGroup(Number of Spawn Groups=2)
   &InitializeParticle(
    Color=<rgba 0.3517 0.4997 1 1>
    Lifetime=0.15
    Mesh Scale Mode=::NewEnumerator1
    Mesh Uniform Scale Max=0.1
    Mesh Uniform Scale Min=0.12
    UsePositionOffset=false
    Write Material Random=true
   )
   &InitialMeshOrientation(Use Rotation Vector=false)
   &AddVelocity(Velocity=<0 0 200>)
  ]
  particleUpdateScript [
   &ParticleState()
   &UpdateMeshOrientation(Rotation Rate=2, Rotation Vector=<0 0 1>)
   &SkeletalMeshLocation(
    Bone / Socket Index=&SpawnGroup()
    Bone Sampling Mode=::NewEnumerator4
    MaskSpawnGroup=false
    Orientation Sampling=::NewEnumerator1
    Position Sampling=::NewEnumerator1
    Sampled Position Offset=&MaskVectorBySpawnGroup(Masked Vector=<12 -10 0>, Passthrough Vector=<-12 10 1>)
    Skeletal Mesh=$User.SkeletalMesh
    Source Movement Scale:Vector3f=<1 1 0.9>
   )
   &SolveForcesAndVelocity(Clamp Velocity=false, Limit Acceleration=false)
   &ScaleMeshSize(Scale Factor=&VectorFromFloat(Value=&FloatFromCurve(FloatCurve=DataInterface() {propertyValues {Curve {keys [{arriveTangent=4.7321, interpMode=RCIM_Cubic, leaveTangent=4.7321, tangentMode=RCTM_User, time=0, value=0}, <at 1 0.7>]}}})))
   &ScaleColor(
    Scale Alpha=&FloatFromCurve(FloatCurve=DataInterface() {propertyValues {Curve {keys [<at 0 1>,<at 1 0>]}}})
    ScaleA=true
    ScaleRGB=true
   )
  ]
  rendererClasses [&NiagaraMeshRendererProperties]
  renderers [{rendererClass=&NiagaraMeshRendererProperties, rendererIndex=0}]
 }
 LightningRibbon {
  emitterSpawnScript [/SetVariables() {isSetParameters=true}]
  emitterUpdateScript [
   &EmitterState()
   /Game/Effects/Niagara/Modules/Ribbons/FN_SpawnRibbon_v3(
    Number of Ribbons=2
    Skeleton Mesh=DataInterface() {propertyValues {
      DefaultMesh=null
      ExcludeBoneName=root
      OptionalComponentClass=null
      PreviewMesh=&M_Med_Soldier_04_Preview
      SkinningMode=SkinOnTheFly
      SoftSourceActor=null
      SourceMode=Default
      UvSetIndex=0
      WholeMeshLOD=-1
      bExcludeBone=false
      bReadDeformedGeometry=true
      bRequireCurrentFrameData=true
      ComponentTags []
      FilteredBones [hand_l,hand_r]
      FilteredSockets []
      MeshUserParameter {parameter {name=None, typeDefHandle {registeredTypeIndex=1}, varData []}}
      SamplingRegions []
     }}
    SpawnRate=30
   )
   /SetVariables(Emitter.Pan:NiagaraFloat=&Add_Float(
     A=$Emitter.Pan
     B=&Multiply_Float(A=$Engine.DeltaTime, B=&Lerp_Float(A=-2, B=-4, Alpha=$System.Dive))
    )) {isSetParameters=true}
  ]
  particleSpawnScript [
   &FN_SetSpawnGroup(Number of Spawn Groups=2)
   /Game/Effects/Niagara/Modules/Ribbons/FN_InitializeRibbon_v5(
    Color:LinearColor=<rgba 17.3438 46.2869 150 1>
    Custom Offset=true
    Enable Skeletal Position=true
    Lifetime=0.23
    Local Vector Offset:Vector3f=<1 0 0>
    Offset Amount=1
    Ribbon Width=65
    SkeletalMesh:NiagaraDataInterfaceSkeletalMesh=$User.SkeletalMesh
   )
   &self:InitializeRibbon_Spin(
    Color:LinearColor=<rgba 17.3438 46.2869 150 1>
    Enable Skeletal Position=true
    Lifetime=0.23
    MeshOffsetDistance=3
    MeshOffsetVector:Vector3f=&MakeVector(
     X=&Sine(
      Normalized Angle=&Add_Float(A=$Emitter.Age, B=&MaskFloatBySpawnGroup(Masked Float=0, Passthrough Float=0.12))
      Period=0.4
      Scale=&MaskFloatBySpawnGroup(Masked Float=-1, Passthrough Float=1)
      Bias=0
     )
     Y=0
     Z=&Cosine(
      Normalized Angle=&Add_Float(A=$Emitter.Age, B=&MaskFloatBySpawnGroup(Masked Float=0, Passthrough Float=0.12))
      Period=0.4
      Scale=1
      Bias=0
     )
    )
    Ribbon Width=65
    SkeletalMesh:NiagaraDataInterfaceSkeletalMesh=$User.SkeletalMesh
   ) {module=InitializeRibbon_Spin}
   &AddVelocity(Velocity=&FN_FrontEndVector3Modifier(Gameplay=<0 0 800>, Front End=<0 0 400>)) {module=AddVelocity001}
   &DynamicMaterialParameters(
    Index 0 Param 1=1
    Index 0 Param 2=&MaskFloatBySpawnGroup(Masked Float=$Emitter.Pan, Passthrough Float=&Add_Float(A=$Emitter.Pan, B=0.5))
    Index 0 Param 3=1
    Index 0 Param 4=&MaskFloatBySpawnGroup(Masked Float=$Emitter.Pan, Passthrough Float=&Add_Float(A=$Emitter.Pan, B=0.5))
    Index 1 Param 1=1
    Index 1 Param 2=1
    Index 1 Param 3=1
    Index 1 Param 4=1
    Index 2 Param 1=1
    Index 2 Param 2=1
    Index 2 Param 3=1
    Index 2 Param 4=1
    Index 3 Param 1=1
    Index 3 Param 2=1
    Index 3 Param 3=1
    Module.Param0WriteEnabled=true
    Module.Param0WriteEnabled1=true
    Module.Param0WriteEnabled2=true
    Module.Param0WriteEnabled3=true
    Module.Param1WriteEnabled=true
    Module.Param1WriteEnabled1=true
    Module.Param1WriteEnabled2=true
    Module.Param1WriteEnabled3=true
    Module.Param2WriteEnabled=true
    Module.Param2WriteEnabled1=true
    Module.Param2WriteEnabled2=true
    Module.Param2WriteEnabled3=true
    Module.Param3WriteEnabled=true
    Module.Param3WriteEnabled1=true
    Module.Param3WriteEnabled2=true
    Module.Param3WriteEnabled3=true
    Write Parameter Index 1=true
    Write Parameter Index 2=true
    Write Parameter Index 3=true
   )
   /SetVariables(Particles.Position=&Add_Vector(
     A=Linked() {linkedVariable {name=Particles.Position, type=NiagaraPosition}}
     B=&Multiply_VectorByFloat(
      Vector=&RotateVectorByQuaternion(
       Vector To Rotate=&MaskVectorBySpawnGroup(Masked Vector=<0 -1 0>, Passthrough Vector=<0 1 0>)
       Quaternion=$Particles.MeshOrientation
      )
      Float=15
     )
    )) {isSetParameters=true}
  ]
  particleUpdateScript [
   &SolveForcesAndVelocity(Clamp Velocity=false, Limit Acceleration=false)
   /Game/Effects/Niagara/Modules/Ribbons/FN_UpdateRibbon_v6(Enable Skeletal Position=true, SkeletalMesh:NiagaraDataInterfaceSkeletalMesh=$User.SkeletalMesh)
   &ScaleRibbonWidth(Ribbon Width Scale=&FloatFromCurve(
     FloatCurve=DataInterface() {propertyValues {Curve {keys [
         {arriveTangent=9.0725, interpMode=RCIM_Cubic, leaveTangent=9.0725, tangentMode=RCTM_User, time=0, value=0}
         <at 0.1408 0.9847 RCIM_Cubic>
         <at 0.5 0.95>
         {arriveTangent=-0.0405, interpMode=RCIM_Cubic, leaveTangent=-0.0405, tangentMode=RCTM_User, time=1, value=0}
        ]}}}
     Scale Curve=0.5
    ))
   &ScaleColor(
    Scale Alpha=&FloatFromCurve(FloatCurve=DataInterface() {propertyValues {Curve {keys [
         {arriveTangent=1.5518, interpMode=RCIM_Cubic, leaveTangent=1.5518, tangentMode=RCTM_User, time=0.1, value=0}
         {arriveTangent=-2.5249, interpMode=RCIM_Cubic, leaveTangent=-2.5249, tangentMode=RCTM_User, time=0.6, value=0.8}
         {arriveTangent=-0.0405, interpMode=RCIM_Cubic, leaveTangent=-0.0405, tangentMode=RCTM_User, time=1, value=0}
        ]}}})
    ScaleA=true
    ScaleRGB=true
   )
  ]
  rendererClasses [&NiagaraRibbonRendererProperties]
  renderers [
   {rendererClass=&NiagaraRibbonRendererProperties}
   {rendererClass=&NiagaraRibbonRendererProperties, rendererIndex=1}
  ]
 }
 LightningSprites {
  emitterUpdateScript [
   &EmitterState()
   &SpawnRate(SpawnRate=10, Use Spawn Probability=false)
  ]
  particleSpawnScript [
   &FN_SetSpawnGroup(Number of Spawn Groups=2)
   &InitializeParticle(
    Color=<rgba 0.2674 0.6356 1 1>
    Lifetime=0.3
    Sprite Size Mode=::NewEnumerator3
    Uniform Sprite Size=75
    UsePositionOffset=false
    Write Material Random=true
   )
   &SkeletalMeshLocation(
    Bone / Socket Index=&SpawnGroup()
    Bone Sampling Mode=::NewEnumerator4
    MaskSpawnGroup=false
    Orientation Sampling=::NewEnumerator1
    Sampled Position Offset=&MaskVectorBySpawnGroup(Masked Vector=<12 -10 0>, Passthrough Vector=<-12 10 1>)
    Skeletal Mesh=$User.SkeletalMesh
   )
   &AddVelocity(Velocity=<0 0 200>)
  ]
  particleUpdateScript [
   &ParticleState()
   &SolveForcesAndVelocity(Clamp Velocity=false, Limit Acceleration=false)
   &ScaleColor(
    Scale Alpha=&FloatFromCurve(FloatCurve=DataInterface() {propertyValues {Curve {keys [<at 0 1>,<at 1 0>]}}})
    ScaleA=true
    ScaleRGB=true
   )
   /SetVariables(Particles.SubImageIndex=&RandomRangeFloat(Minimum=0, Maximum=15, Evaluation Type=::NewEnumerator1)) {isSetParameters=true}
  ]
  rendererClasses [&NiagaraSpriteRendererProperties]
  renderers [{rendererClass=&NiagaraSpriteRendererProperties}]
 }
 LightningSpritesAttached {
  emitterUpdateScript [
   &EmitterState()
   &SpawnRate(SpawnRate=20, Use Spawn Probability=false)
  ]
  particleSpawnScript [
   &FN_SetSpawnGroup(Number of Spawn Groups=2)
   &InitializeParticle(
    Color=<rgba 1.3371 3.178 5 1>
    Lifetime=0.2
    Sprite Rotation Mode=::NewEnumerator2
    Sprite Size Mode=::NewEnumerator1
    Uniform Sprite Size Max=35
    Uniform Sprite Size Min=45
    UsePositionOffset=false
    Write Material Random=true
   )
   &CameraOffset(Camera Offset Amount=-20)
   &DynamicMaterialParameters(
    Index 0 Param 1=&RandomRangeFloat(Minimum=2.5, Maximum=3.5)
    Index 0 Param 2=&FloatFromCurve(FloatCurve=DataInterface() {propertyValues {Curve {keys [<at 0 1>,<at 1 -1>]}}})
    Index 0 Param 3=&RandomRangeFloat(Minimum=0.2, Maximum=0.7)
    Module.Param0WriteEnabled=true
    Module.Param1WriteEnabled=true
    Module.Param2WriteEnabled=true
    Module.Param3WriteEnabled=true
   )
  ]
  particleUpdateScript [
   &ParticleState()
   &SkeletalMeshLocation(
    Bone / Socket Index=&SpawnGroup()
    Bone Sampling Mode=::NewEnumerator4
    MaskSpawnGroup=false
    Orientation Sampling=::NewEnumerator1
    Position Sampling=::NewEnumerator1
    Sampled Position Offset=&Add_Vector(
     A=&MaskVectorBySpawnGroup(Masked Vector=<20 0 0>, Passthrough Vector=<-20 0 1>)
     B=&RandomVector(Vector Scale=15)
    )
    Skeletal Mesh=$User.SkeletalMesh
   )
   &ScaleColor(
    Scale Alpha=&FloatFromCurve(FloatCurve=DataInterface() {propertyValues {Curve {keys [<at 0 1>,<at 1 0>]}}})
    ScaleA=true
    ScaleRGB=true
   )
  ]
  rendererClasses [&NiagaraSpriteRendererProperties]
  renderers [{rendererClass=&NiagaraSpriteRendererProperties}]
 }
 SparkMesh {
  emitterUpdateScript [
   &EmitterState()
   &SpawnRate(SpawnRate=&FN_FrontEndFloatModifier(Gameplay=50, Front End=10), Use Spawn Probability=false)
  ]
  particleSpawnScript [
   &FN_SetSpawnGroup(Number of Spawn Groups=2)
   &InitializeParticle(
    Color=<rgba 5 0.3498 0 1>
    Lifetime Max=0.35
    Lifetime Min=0.15
    Lifetime Mode=::NewEnumerator1
    Mesh Scale=&Multiply_Vector(
     A=&RandomRangeVector(Minimum=<0.04 0.04 0.06>, Maximum=<0.08 0.08 0.06>)
     B=&FN_FrontEndVector3Modifier(Gameplay=<1 1 1>, Front End=<1 1 0.6>)
    )
    Mesh Scale Mode=::NewEnumerator4
    UsePositionOffset=false
    Write Material Random=true
   )
   &SkeletalMeshLocation(
    Bone / Socket Index=&SpawnGroup()
    Bone Sampling Mode=::NewEnumerator4
    MaskSpawnGroup=false
    Orientation Sampling=::NewEnumerator1
    Sampled Position Offset=&Add_Vector(A=&MaskVectorBySpawnGroup(Masked Vector=<10 -12 0>, Passthrough Vector=<-10 12 1>), B=<0 0 0>)
    Skeletal Mesh=$User.SkeletalMesh
   )
   &InitialMeshOrientation(
    Mesh Orientation Mode=::NewEnumerator2
    Rotation=&RandomRangeVector(Minimum=<0.1 0 0>, Maximum=<0.1 0 0>)
    Use Rotation Vector=true
   )
   &AddVelocity(Velocity=<0 0 450>)
   /SetVariables(Particles.MeshOrientation=&MultiplyQuaternion(
     Quaternion A=&MakeQuatFromAxisAngle(Axis=<0 0 1>, AngleInDegrees=&RandomRangeFloat(Minimum=0, Maximum=360))
     Quaternion B=$Particles.MeshOrientation
    )) {isSetParameters=true}
  ]
  particleUpdateScript [
   &ParticleState()
   &Drag(Drag=3, Use Linear Drag=true, Use Rotational Drag=false)
   &SolveForcesAndVelocity(Clamp Velocity=false, Limit Acceleration=false)
   &UpdateMeshOrientation(Rotation Coordinate Space=::NewEnumerator0, Rotation Rate=4, Rotation Vector=<0 0 1>)
   &ScaleMeshSize(Scale Factor=&VectorFromCurve(VectorCurve=DataInterface() {propertyValues {XCurve {keys [<at 0 1>,<at 1 0.7>]}, YCurve {keys [<at 0 1>,<at 1 0.7>]}, ZCurve {keys [<at 0 1>,<at 1 0.5>]}}}))
   &ScaleColor(
    Scale Alpha=&FloatFromCurve(FloatCurve=DataInterface() {propertyValues {Curve {keys [{tangentMode=RCTM_User, time=0, value=1}, {tangentMode=RCTM_User, time=1, value=0}]}}})
    ScaleA=true
    ScaleRGB=true
   )
  ]
  rendererClasses [&NiagaraMeshRendererProperties]
  renderers [{rendererClass=&NiagaraMeshRendererProperties, rendererIndex=0}]
 }
 Sparks {
  bEnabled=false
  simTarget=GPUComputeSim
  emitterUpdateScript [
   &EmitterState()
   &SpawnRate(
    SpawnGroup:NiagaraInt32=&RandomRangeInt(Minimum=0, Maximum=1)
    SpawnRate=500
    Use Spawn Probability=false
   )
  ]
  particleSpawnScript [
   &FN_SetSpawnGroup(Number of Spawn Groups=2)
   &InitializeParticle(
    Color Maximum=<rgba 150 20.5059 0 1>
    Color Minimum=<rgba 50 6.8353 0 1>
    Color Mode=::NewEnumerator2
    Lifetime Max=0.25
    Lifetime Min=0.1
    Lifetime Mode=::NewEnumerator1
    Sprite Size Max=<5 20>
    Sprite Size Min=<3 15>
    Sprite Size Mode=::NewEnumerator2
    UsePositionOffset=false
    Write Material Random=true
   ) {module=InitializeParticle001}
   &SkeletalMeshLocation(
    Bone / Socket Index=&SpawnGroup()
    Bone Sampling Mode=::NewEnumerator4
    MaskSpawnGroup=false
    Orientation Sampling=::NewEnumerator1
    Sampled Position Offset=&Add_Vector(A=&MaskVectorBySpawnGroup(Masked Vector=<10 -12 0>, Passthrough Vector=<-10 12 1>), B=<0 0 0>)
    Skeletal Mesh=$User.SkeletalMesh
   )
   &ShapeLocation(
    Additional Quaternion Rotation=true
    Additional Yaw / Pitch / Roll=<0 0.25 0.25>
    Cylinder Height=10
    Cylinder Radius=15
    MaskSpawnGroup=false
    Override Seed=false
    Rotation Mode=::NewEnumerator3
    Rotation Quaternion=$Particles.SkeletalMeshLocation.SampledRotation
    Shape Primitive=::NewEnumerator1
    Surface Only=true
    Use Endcaps In Surface Only Mode=false
    UseOwnerScale=true
   )
   &SubUVAnimation(End Frame=3, SubUV Animation Mode:ENiagara_SubUVLookupMode=::NewEnumerator2)
  ]
  particleUpdateScript [
   &ParticleState()
   &SkeletalMeshLocation(
    Bone / Socket Index=&SpawnGroup()
    Bone Sampling Mode=::NewEnumerator4
    MaskSpawnGroup=false
    Orientation Sampling=::NewEnumerator1
    Position Sampling=::NewEnumerator2
    Sampled Position Offset=&Add_Vector(A=&MaskVectorBySpawnGroup(Masked Vector=<10 -12 0>, Passthrough Vector=<-10 12 1>), B=<0 0 0>)
    Skeletal Mesh=$User.SkeletalMesh
   ) {module=SkeletalMeshLocation001}
   /Niagara/Modules/Update/Velocity/VortexVelocity(
    Module.Use Influence Falloff=false
    Velocity Amount=350
    Vortex Axis:Vector3f=&RotateVectorByQuaternion(Vector To Rotate=<0 1 0>, Quaternion=$Particles.SkeletalMeshLocation001.SampledRotation)
    Vortex Origin:NiagaraPosition=$Particles.SkeletalMeshLocation001.SampledPosition
   )
   &PointAttractionForce(
    Attraction Radius=1500
    Attraction Strength=&RandomRangeFloat(Minimum=-100, Maximum=0)
    Attractor Position:NiagaraPosition=$Particles.SkeletalMeshLocation001.SampledPosition
    Kill Within Radius=false
    Module.Use Falloff=false
    UseAttractorOffset=false
   )
   &GravityForce(Gravity=&FN_FrontEndVector3Modifier(
     Gameplay=<0 0 2500>
     Front End=&MakeVector(
      X=0
      Y=0
      Z=&FloatFromCurve(FloatCurve=DataInterface() {propertyValues {Curve {keys [<at 0 1>,<at 1 35000>]}}})
     )
    ))
   &Drag(Use Linear Drag=true, Use Rotational Drag=false)
   &SolveForcesAndVelocity(Clamp Velocity=false, Limit Acceleration=false)
   &ScaleColor(
    Scale Alpha=&FloatFromCurve(FloatCurve=DataInterface() {propertyValues {Curve {keys [
         <at 0 0 RCIM_Cubic>
         {arriveTangent=1.6041, leaveTangent=1.6041, time=0.05, value=1.0113}
         {interpMode=RCIM_Cubic, leaveTangent=-1.0633, time=0.75, value=1}
         <at 0.99 0 RCIM_Cubic>
        ]}}})
    ScaleA=true
    ScaleRGB=true
   )
   &ScaleSpriteSize(Scale Factor=&Vector2DFromFloat(Value=&FloatFromCurve(
      FloatCurve=DataInterface() {propertyValues {Curve {keys [
          {arriveTangent=-0.4919, interpMode=RCIM_Cubic, leaveTangent=-0.4919, tangentMode=RCTM_User, time=0.5, value=1}
          {arriveTangent=-3.3301, interpMode=RCIM_Cubic, leaveTangent=-3.3302, tangentMode=RCTM_User, time=1, value=0}
         ]}}}
      Scale Curve=0.65
     )))
  ]
  rendererClasses [&NiagaraSpriteRendererProperties]
  renderers [{rendererClass=&NiagaraSpriteRendererProperties}]
 }
 SparksVortex {
  simTarget=GPUComputeSim
  emitterUpdateScript [
   &EmitterState()
   &SpawnRate(SpawnRate=45, Use Spawn Probability=false)
   &SpawnRate(SpawnGroup=1, SpawnRate=45, Use Spawn Probability=false) {module=SpawnRate001}
  ]
  particleSpawnScript [
   &InitializeParticle(
    Color=<rgba 35 1.8898 0 1>
    Lifetime Max=0.3
    Lifetime Min=0.2
    Lifetime Mode=::NewEnumerator1
    Sprite Rotation Angle=90
    Sprite Rotation Mode=::NewEnumerator0
    Sprite Size=&FN_FrontEndVector2Modifier(Front End:Vector2f=<2 5>, Gameplay:Vector2f=&RandomRangeVector2D(Minimum=<2 5>, Maximum=<4 10>))
    Sprite Size Mode=::NewEnumerator4
    UsePositionOffset=false
    Write Material Random=true
   )
   &SkeletalMeshLocation(
    Bone / Socket Index=&SpawnGroup()
    Bone Sampling Mode=::NewEnumerator4
    MaskSpawnGroup=false
    Sampled Position Offset=&MakeVector(
     X=&Sine(
      Normalized Angle=$Emitter.Age
      Period=0.5
      Scale=&FN_FrontEndFloatModifier(Gameplay=-20, Front End=-10)
      Bias=0
     )
     Y=&Cosine(
      Normalized Angle=$Emitter.Age
      Period=0.5
      Scale=&FN_FrontEndFloatModifier(Gameplay=20, Front End=10)
      Bias=0
     )
     Z=0
    )
    Sampled Position Offset Coordinate Space:ENiagaraSkelMeshTransforms=::NewEnumerator1
    Skeletal Mesh=$User.SkeletalMesh
   )
   &AddVelocity(Velocity=&RandomRangeVector(Minimum=<-50 -50 300>, Maximum=<50 50 375>))
   &SubUVAnimation(End Frame=3, SubUV Animation Mode:ENiagara_SubUVLookupMode=::NewEnumerator2)
  ]
  particleUpdateScript [
   &ParticleState()
   &GravityForce(Gravity=<0 0 -150>)
   &SolveForcesAndVelocity(Clamp Velocity=false, Limit Acceleration=false)
   &ScaleSpriteSize(Uniform Curve Sprite Scale=DataInterface() {propertyValues {Curve {keys [
        {tangentMode=RCTM_User, time=0, value=1}
        {arriveTangent=-0.0405, interpMode=RCIM_Cubic, leaveTangent=-0.0405, tangentMode=RCTM_User, time=0.7009, value=1}
        {arriveTangent=-8.0089, interpMode=RCIM_Cubic, leaveTangent=-8.0089, tangentMode=RCTM_User, time=1, value=0}
       ]}}})
  ]
  rendererClasses [&NiagaraSpriteRendererProperties]
  renderers [{rendererClass=&NiagaraSpriteRendererProperties}]
 }
 Sparks_Optimized {
  simTarget=GPUComputeSim
  emitterSpawnScript [/SetVariables() {isSetParameters=true}]
  emitterUpdateScript [
   &EmitterState()
   &SpawnRate(
    SpawnGroup:NiagaraInt32=&RandomRangeInt(Minimum=0, Maximum=1)
    SpawnRate=500
    Use Spawn Probability=false
   )
   &self:ScratchModule(Offset:Vector3f=<10 -12 0>, Skeletal Mesh:NiagaraDataInterfaceSkeletalMesh=$User.SkeletalMesh) {module=ScratchModule}
   &self:ScratchModule(Bone Index=1, Offset:Vector3f=<-10 12 1>, Skeletal Mesh:NiagaraDataInterfaceSkeletalMesh=$User.SkeletalMesh) {module=ScratchModule001}
   /SetVariables(
    Emitter.Bone0Position:NiagaraPosition=$Output.ScratchModule.Position
    Emitter.Bone0Rotation:Quat4f=$Output.ScratchModule.Rotation
    Emitter.Bone1Position:NiagaraPosition=$Output.ScratchModule001.Position
    Emitter.Bone1Rotation:Quat4f=$Output.ScratchModule001.Rotation
   ) {isSetParameters=true}
  ]
  particleSpawnScript [
   &FN_SetSpawnGroup(Number of Spawn Groups=2)
   &InitializeParticle(
    Color Maximum=<rgba 150 20.5059 0 1>
    Color Minimum=<rgba 50 6.8353 0 1>
    Color Mode=::NewEnumerator2
    Lifetime Max=0.25
    Lifetime Min=0.1
    Lifetime Mode=::NewEnumerator1
    Sprite Size Max=<5 20>
    Sprite Size Min=<3 15>
    Sprite Size Mode=::NewEnumerator2
    UsePositionOffset=false
    Write Material Random=true
   ) {module=InitializeParticle001}
   /SetVariables(Particles.Position=&self:SelectVector(
     A:NiagaraPosition=$Emitter.Bone0Position
     B:NiagaraPosition=$Emitter.Bone1Position
     SpawnGroup:NiagaraInt32=&SpawnGroup()
    )) {isSetParameters=true}
   &ShapeLocation(
    Additional Quaternion Rotation=true
    Additional Yaw / Pitch / Roll=<0 0.25 0.25>
    Cylinder Height=10
    Cylinder Radius=15
    MaskSpawnGroup=false
    Override Seed=false
    Rotation Mode=::NewEnumerator3
    Rotation Quaternion=&self:SelectRotation(A:Quat4f=$Emitter.Bone0Rotation, B:Quat4f=$Emitter.Bone1Rotation, SpawnGroup:NiagaraInt32=$Particles.SpawnGroup)
    Shape Primitive=::NewEnumerator1
    Surface Only=true
    Use Endcaps In Surface Only Mode=false
    UseOwnerScale=true
   )
   &SubUVAnimation(End Frame=3, SubUV Animation Mode:ENiagara_SubUVLookupMode=::NewEnumerator2)
  ]
  particleUpdateScript [
   &ParticleState()
   &VortexVelocity(
    Module.Use Influence Falloff=false
    Velocity Amount=350
    Vortex Axis:Vector3f=&RotateVectorByQuaternion(
     Vector To Rotate=<0 1 0>
     Quaternion=&self:SelectRotation(A:Quat4f=$Emitter.Bone0Rotation, B:Quat4f=$Emitter.Bone1Rotation, SpawnGroup:NiagaraInt32=$Particles.SpawnGroup)
    )
    Vortex Origin:Vector3f=&self:SelectVector(A:NiagaraPosition=$Emitter.Bone0Position, B:NiagaraPosition=$Emitter.Bone1Position, SpawnGroup:NiagaraInt32=$Particles.SpawnGroup)
   )
   &PointAttractionForce(
    Attraction Radius=1500
    Attraction Strength=&RandomRangeFloat(Minimum=-100, Maximum=0)
    Attractor Position:NiagaraPosition=&self:SelectVector(A:NiagaraPosition=$Emitter.Bone0Position, B:NiagaraPosition=$Emitter.Bone1Position, SpawnGroup:NiagaraInt32=$Particles.SpawnGroup)
    Kill Within Radius=false
    Module.Use Falloff=false
    UseAttractorOffset=false
   )
   &GravityForce(Gravity=&FN_FrontEndVector3Modifier(
     Gameplay=<0 0 2500>
     Front End=&MakeVector(
      X=0
      Y=0
      Z=&FloatFromCurve(FloatCurve=DataInterface() {propertyValues {Curve {keys [<at 0 1>,<at 1 35000>]}}})
     )
    ))
   &Drag(Use Linear Drag=true, Use Rotational Drag=false)
   &SolveForcesAndVelocity(Clamp Velocity=false, Limit Acceleration=false)
   &ScaleColor(
    Scale Alpha=&FloatFromCurve(FloatCurve=DataInterface() {propertyValues {Curve {keys [
         <at 0 0 RCIM_Cubic>
         {arriveTangent=1.6041, leaveTangent=1.6041, time=0.05, value=1.0113}
         {interpMode=RCIM_Cubic, leaveTangent=-1.0633, time=0.75, value=1}
         <at 0.99 0 RCIM_Cubic>
        ]}}})
    ScaleA=true
    ScaleRGB=true
   )
   &ScaleSpriteSize(Scale Factor=&Vector2DFromFloat(Value=&FloatFromCurve(
      FloatCurve=DataInterface() {propertyValues {Curve {keys [
          {arriveTangent=-0.4919, interpMode=RCIM_Cubic, leaveTangent=-0.4919, tangentMode=RCTM_User, time=0.5, value=1}
          {arriveTangent=-3.3301, interpMode=RCIM_Cubic, leaveTangent=-3.3302, tangentMode=RCTM_User, time=1, value=0}
         ]}}}
      Scale Curve=0.65
     )))
  ]
  rendererClasses [&NiagaraSpriteRendererProperties]
  renderers [{rendererClass=&NiagaraSpriteRendererProperties}]
 }
}
summary {
 systemName=NS_Contrail_OrderGuard
 emitters [
  {emitterName=Blades, rendererClasses [&NiagaraMeshRendererProperties, &NiagaraLightRendererProperties]}
  {emitterName=LightningRibbon, rendererClasses [&NiagaraRibbonRendererProperties]}
  {emitterName=LightningBalls, rendererClasses [&NiagaraMeshRendererProperties]}
  {emitterName=LightningSprites, rendererClasses [&NiagaraSpriteRendererProperties]}
  {emitterName=LightningSpritesAttached, rendererClasses [&NiagaraSpriteRendererProperties]}
  {bEnabled=false, emitterName=Sparks, simTarget=GPUComputeSim, rendererClasses [&NiagaraSpriteRendererProperties]}
  {emitterName=SparkMesh, rendererClasses [&NiagaraMeshRendererProperties]}
  {emitterName=SparksVortex, simTarget=GPUComputeSim, rendererClasses [&NiagaraSpriteRendererProperties]}
  {emitterName=Sparks_Optimized, simTarget=GPUComputeSim, rendererClasses [&NiagaraSpriteRendererProperties]}
 ]
 userVariables [
  User.bIsFrontEnd {type=NiagaraBool}
  User.bIsFrontEndPreview {type=NiagaraBool}
  User.PlayerSpeed {type=NiagaraFloat, defaultValue=0, description=""}
  User.RibbonIDOffset {type=NiagaraInt32}
  User.SkeletalMesh {
   type=NiagaraDataInterfaceSkeletalMesh
   defaultValue {
    struct=NiagaraExt_VariableValue_DataInterface
    value {dataInterface=&self:NiagaraDataInterfaceSkeletalMesh_0}
   }
  }
 ]
}
